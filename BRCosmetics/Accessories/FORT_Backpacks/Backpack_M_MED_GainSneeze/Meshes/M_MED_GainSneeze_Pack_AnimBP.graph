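# Animation Blueprint for the GainSneeze backpack cosmetic that plays its idle animation.
# A single-state machine, Animation Testing, outputs GainSneeze_Idle_pack; a leftover Random Sequence Player referencing a CoconutShell idle sits unconnected in that state.
# A Modify Bone node then shrinks the pack bone with an additive-mode 0.85 scale before the output pose.
# The event graph holds only unconnected default nodes, with no variables or logic.
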AnimBlueprint /BRCosmetics/Accessories/FORT_Backpacks/Backpack_M_MED_GainSneeze/Meshes/M_MED_GainSneeze_Pack_AnimBP
target_skeleton=&/M_MED_GainSneeze_Pack_Skeleton
graphs [
 AnimGraph {nodes [
   AnimGraphNode_StateMachine_0 = Animation Testing() {is=title AnimGraphNode_StateMachine}
   n1 = "Transform (Modify) Bone - Bone: pack"(ComponentPose=Local To Component(LocalPose=@AnimGraphNode_StateMachine_0), Translation="0,0,0", Rotation="0,0,0", Scale="0.85,0.85,0.85") {
    rotationMode=BMM_Additive
    rotationSpace=BCS_BoneSpace
    scaleMode=BMM_Additive
    alphaBoolBlend {blendOption=Linear}
    boneToModify {boneName=pack}
    scale=<0.85 0.85 0.85>
   }
   Output Pose(Result=Component To Local(ComponentPose=@n1)) {layerGroup=None}
  ]}
 Animation Testing {
  entry=State
  path=AnimGraph.AnimGraphNode_StateMachine_0.Animation Testing
  ids=id
  nodes [AnimStateNode_0 = State() {is=title AnimStateNode}]
 }
 State {
  path=AnimGraph.AnimGraphNode_StateMachine_0.Animation Testing.AnimStateNode_0.State
  nodes [
   Random Sequence Player!() -> (Pose:Pose Link Structure) {
    is=title AnimGraphNode_RandomPlayer
    entries [
     {
      chanceToPlay=0.5
      maxLoopCount=0
      maxPlayRate=1
      minLoopCount=0
      minPlayRate=1
      sequence=/BRCosmetics/Accessories/FORT_Backpacks/Backpack_M_MED_CoconutShell/Anim/CoconutShell_Idle_pack
      blendIn {blendOption=Linear, blendTime=0.2, customCurve=null}
     }
     {
      chanceToPlay=0.5
      maxLoopCount=0
      maxPlayRate=1
      minLoopCount=0
      minPlayRate=1
      sequence=null
      blendIn {blendOption=Linear, blendTime=0.2, customCurve=null}
     }
    ]
   }
   Output Animation Pose(Result=Sequence Player 'GainSneeze_Idle_pack'() {sequence=/BRCosmetics/Accessories/FORT_Backpacks/Backpack_M_MED_GainSneeze/Anim/GainSneeze_Idle_pack})
  ]
 }
 EventGraph {nodes [
   TryGetPawnOwner()
   Event BlueprintUpdateAnimation()
  ]}
]
variables []
